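; Parent Blueprint for an EQ block item in Playgrounds that can be enabled or disabled and fades its mesh when player pawns overlap it.
; While enabled, FXOverlapCube overlaps drive an opacity timeline on a lazily created dynamic material, with visibility toggles and original material restoration.
; Gameplay message receivers set bEnabledState and collision, while its OnRep updates visibility and collision.
; The block turns on at BeginPlay and minigame end, and applies bEnabledState at minigame start.
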(event OnComponentBeginOverlap(FXOverlapCube)
  (params (OverlappedComponent :type Primitive%20Component%20Object%20Reference) (OtherActor :type Actor%20Object%20Reference) (OtherComp :type Primitive%20Component%20Object%20Reference) (OtherBodyIndex :type Integer) (bFromSweep :type Boolean) (SweepResult :type Hit%20Result%20Structure%20%28by%20ref%29))
  (if (Variables|Default|GetbEnabledState)
    (Networking|CanExecuteCosmeticEvents
      (:True
        (Utilities|Casting|CastToFortPlayerPawn OtherActor)
        (CallFunction|GetMeshMID)
        (bind (_direction _lerp) (Timeline|OpacityTL Play)
          (:Update
            (Rendering|Material|SetScalarParameterValue (Variables|Default|GetMeshMID) "Opacity" _lerp :!class /Script/Engine.MaterialInstanceDynamic))
          (:Finished
            _)))))
  (switch Utilities|FlowControl|Switch|SwitchonETimelineDirection _direction
    (:Forward
      (Rendering|SetVisibility (Variables|Debug|GetStaticMeshComponent) false false))
    (:Backward
      (bind _originalmaterial (Variables|Default|GetOriginalMaterial)
        (:then
          (Rendering|Material|SetMaterial (Variables|Debug|GetStaticMeshComponent) 0 _originalmaterial))
        (:else)))))

(event OnComponentEndOverlap(FXOverlapCube)
  (params (OverlappedComponent :type Primitive%20Component%20Object%20Reference) (OtherActor :type Actor%20Object%20Reference) (OtherComp :type Primitive%20Component%20Object%20Reference) (OtherBodyIndex :type Integer))
  (if (Variables|Default|GetbEnabledState)
    (Networking|CanExecuteCosmeticEvents
      (:True
        (Utilities|Casting|CastToFortPlayerPawn OtherActor)
        (CallFunction|GetMeshMID)
        (Rendering|SetVisibility (Variables|Debug|GetStaticMeshComponent) true false)
        (Timeline|OpacityTL Reverse)))
    (else
      (switch Utilities|FlowControl|Switch|SwitchonETimelineDirection (stale OnComponentBeginOverlap(FXOverlapCube) _direction)
        (:Forward
          (Rendering|SetVisibility (Variables|Debug|GetStaticMeshComponent) false false))
        (:Backward
          (bind _originalmaterial (Variables|Default|GetOriginalMaterial)
            (:then
              (Rendering|Material|SetMaterial (Variables|Debug|GetStaticMeshComponent) 0 _originalmaterial))
            (:else)))))))

(event EventBeginPlay
  (|CallParentFunction)
  (CallFunction|SetEnableState true))

(event MinigameStartedDelegate(FortMinigameProgress)
  ;[1 0.604 0.9646] Settings based on enabled options bool when starting/ending minigame
  (CallFunction|SetEnableState (Variables|Default|GetbEnabledState)))

(event MinigameEndedDelegate(FortMinigameProgress)
  ;[1 0.604 0.9646] Settings based on enabled options bool when starting/ending minigame
  (CallFunction|SetEnableState true))

(event Custom|SetEnableState
  (params (Enabled :type Boolean))
  (bind _staticmeshcomponent (Variables|Debug|GetStaticMeshComponent))
  ;[1 0.604 0.9646] Settings based on enabled options bool when starting/ending minigame
  (Rendering|SetVisibility (Variables|Debug|GetStaticMeshComponent) Enabled true)
  (if Enabled
    (Collision|SetCollisionEnabled _staticmeshcomponent "QueryAndPhysics")
    (else
      (Collision|SetCollisionEnabled _staticmeshcomponent "NoCollision"))))

(event OnGameplayMessageReceived(DisableWhenReceived)
  (params (TriggerInstigator :type Controller%20Object%20Reference))
  (bind _staticmeshcomponent (Variables|Debug|GetStaticMeshComponent))
  ;[0.6249 1 0.9901] Enabled Message Receivers
  (Variables|Default|SetbEnabledState false)
  (if (Variables|Default|GetbEnabledState)
    (Collision|SetCollisionEnabled _staticmeshcomponent "QueryAndPhysics")
    (else
      (Collision|SetCollisionEnabled _staticmeshcomponent "NoCollision"))))

(event OnGameplayMessageReceived(EnableWhenReceived)
  (params (TriggerInstigator :type Controller%20Object%20Reference))
  (bind _staticmeshcomponent (Variables|Debug|GetStaticMeshComponent))
  ;[0.6249 1 0.9901] Enabled Message Receivers
  (Variables|Default|SetbEnabledState true)
  (if (Variables|Default|GetbEnabledState)
    (Collision|SetCollisionEnabled _staticmeshcomponent "QueryAndPhysics")
    (else
      (Collision|SetCollisionEnabled _staticmeshcomponent "NoCollision"))))

(fn GetMeshMID
  (bind _staticmeshcomponent (Variables|Debug|GetStaticMeshComponent))
  (bind _returnvalue (Rendering|Material|GetMaterial _staticmeshcomponent 0))
  (bind _asmaterial_instance_dynamic (Utilities|Casting|CastToMaterialInstanceDynamic _returnvalue)
    (:then
      (Variables|Default|SetMeshMID _asmaterial_instance_dynamic))
    (:CastFailed
      (Variables|Default|SetOriginalMaterial _returnvalue)
      (bind _returnvalue_2 (Rendering|Material|CreateDynamicMaterialInstance (Variables|Debug|GetStaticMeshComponent) 0 0 "None" :!class /Script/Engine.PrimitiveComponent))
      (Variables|Default|SetMeshMID _returnvalue_2))))

(fn OnRep_bEnabledState
  (bind _benabledstate (Variables|Default|GetbEnabledState))
  (bind _staticmeshcomponent (Variables|Debug|GetStaticMeshComponent))
  (Networking|CanExecuteCosmeticEvents
    (:True
      (Rendering|SetVisibility (Variables|Debug|GetStaticMeshComponent) _benabledstate true)
      (if _benabledstate
        (Collision|SetCollisionEnabled _staticmeshcomponent "QueryAndPhysics")
        (else
          (Collision|SetCollisionEnabled _staticmeshcomponent "NoCollision"))))))

(fn ConstructionScript)
