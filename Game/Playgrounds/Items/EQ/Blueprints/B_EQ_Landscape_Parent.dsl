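; Base Blueprint for EQ landscape items in Playgrounds: a static mesh with trigger-only interaction collision that can be shown or hidden during minigames.
; Enable/disable gameplay messages set bEnabledState, whose RepNotify updates mesh visibility.
; Minigame start applies that state, and minigame end makes the mesh visible again.
; Construction creates two dynamic materials, assigned on BeginPlay or by an editor event and returned by the material override function.
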(event EventBeginPlay
  (Collision|SetCollisionProfileName (Variables|Debug|GetStaticMeshComponent) "FortTriggerOnlyInteractions" true :!class /Script/Engine.PrimitiveComponent)
  (Networking|CanExecuteCosmeticEvents
    (:True
      (Rendering|Material|SetMaterial (Variables|Debug|GetStaticMeshComponent) 0 (Variables|Default|GetMeshMID_0))
      (Rendering|Material|SetMaterial (Variables|Debug|GetStaticMeshComponent) 1 (Variables|Default|GetMeshMID_1)))))

(event Custom|AssignMIDInEditor
  (Rendering|Material|SetMaterial (Variables|Debug|GetStaticMeshComponent) 0 (Variables|Default|GetMeshMID_0))
  (Rendering|Material|SetMaterial (Variables|Debug|GetStaticMeshComponent) 1 (Variables|Default|GetMeshMID_1)))

(event MinigameStartedDelegate(FortMinigameProgress)
  (CallFunction|SetEnableState (Variables|Default|GetbEnabledState)))

(event MinigameEndedDelegate(FortMinigameProgress)
  (CallFunction|SetEnableState true))

(event Custom|SetEnableState
  (params (Enabled :type Boolean))
  (Rendering|SetVisibility (Variables|Debug|GetStaticMeshComponent) Enabled true))

(event OnGameplayMessageReceived(EnabledOnReceiving)
  (params (TriggerInstigator :type Controller%20Object%20Reference))
  (Variables|Default|SetbEnabledState true))

(event OnGameplayMessageReceived(DisableOnReceiving)
  (params (TriggerInstigator :type Controller%20Object%20Reference))
  (Variables|Default|SetbEnabledState false))

(fn GetOverrideMeshMaterial
  (params (MeshComp :type Static%20Mesh%20Component%20Object%20Reference) (MatIdx :type Integer))
  (if (Utilities|Equal(Object) MeshComp (Variables|Debug|GetStaticMeshComponent))
    (return (Utilities|Select :Index MatIdx :Option%200 (Variables|Default|GetMeshMID_0) :Option%201 (Variables|Default|GetMeshMID_1)))
    (else
      (return (|CallParentFunction MeshComp MatIdx)))))

(fn IsInPlayMode
  (bind _currentminigame (Variables|/Script/FortniteGame.FortMinigameLogicComponent|GetCurrentMinigame (Variables|Default|GetFortMinigameProgress))
    (:then
      (return (Minigame|IsinPlayMode _currentminigame :!class /Script/FortniteGame.FortMinigame)))
    (:else
      (return false))))

(fn OnRep_bEnabledState
  (Networking|CanExecuteCosmeticEvents
    (:True
      (Rendering|SetVisibility (Variables|Debug|GetStaticMeshComponent) (Variables|Default|GetbEnabledState) true))))

(fn ConstructionScript
  (bind _staticmeshcomponent (Variables|Debug|GetStaticMeshComponent))
  (Networking|CanExecuteCosmeticEvents
    (:True
      (bind _returnvalue (Rendering|Material|CreateDynamicMaterialInstance _staticmeshcomponent 0 (Variables|Default|GetMaterialIndex_0) "None" :!class /Script/Engine.PrimitiveComponent))
      (Variables|Default|SetMeshMID_0 _returnvalue)
      (bind _returnvalue_2 (Rendering|Material|CreateDynamicMaterialInstance _staticmeshcomponent 1 (Variables|Default|GetMaterialIndex_1) "None" :!class /Script/Engine.PrimitiveComponent))
      (Variables|Default|SetMeshMID_1 _returnvalue_2))))
