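# MF_MusicDebugNoise is a material function that generates an animated noise value, apparently as debug input for music-reactive EQ materials.
# A Panner scrolls the UVs at the DebugNoisePanSpeed input, which defaults to a same-named vector parameter's RG, and samples T_Noise01_Inverse, outputting red as Noise.
# Near-zero U and zero V tiling collapse the UVs, so the noise varies over time while staying practically uniform across the surface.
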MaterialFunction /Game/Playgrounds/Items/EQ/Materials/MF_MusicDebugNoise
graph {nodes [
  n1 = TextureCoordinate() {uTiling=0.0001, vTiling=0}
  n2 = FunctionInput(Preview=ComponentMask(None=VectorParameter() {parameterName=DebugNoisePanSpeed, defaultValue=<rgba 0.2 0.05 0 0>}.RGB) {g=true, r=true}) {bUsePreviewValueAsDefault=true, inputName=DebugNoisePanSpeed, inputType=FunctionInput_Vector2}
  FunctionOutput(None=TextureSample(UVs=Panner(Coordinate=@n1, Speed=@n2) {bFractionalPart=true, speedX=0.2, speedY=0.05}) {samplerType=SAMPLERTYPE_LinearColor, texture=/Game/Global/Textures/Default/Noise/T_Noise01_Inverse}.R) {outputName=Noise}
 ]}
