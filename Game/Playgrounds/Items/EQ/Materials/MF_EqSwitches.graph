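# Material function that outputs one audio EQ band level from the Creative Jukebox material parameter collection, for EQ item materials.
# Nested static switches in the AudioData group pick the band at compile time, with priority High, MidHigh, Mid, LowMid, then Low, falling back to a default constant.
# An outer UseDebugNoise switch overrides the band with MF_MusicDebugNoise output.
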MaterialFunction /Game/Playgrounds/Items/EQ/Materials/MF_EqSwitches
graph {nodes [
  n1 = CollectionParameter() {collection=/Game/Creative/Devices/Jukebox/Creative_Jukebox_MPC, parameterName=High}
  n2 = CollectionParameter() {collection=&Creative_Jukebox_MPC, parameterName=LowMid}
  n3 = StaticSwitchParameter(True=CollectionParameter() {collection=&Creative_Jukebox_MPC, parameterName=Low}, False=Constant()) {group=AudioData, parameterName=Low}
  n4 = StaticSwitchParameter(
   True=CollectionParameter() {collection=&Creative_Jukebox_MPC, parameterName=MidHigh}
   False=StaticSwitchParameter(
    True=CollectionParameter() {collection=&Creative_Jukebox_MPC, parameterName=Mid}
    False=StaticSwitchParameter(True=@n2, False=@n3) {group=AudioData, parameterName=LowMid, sortPriority=1}
   ) {group=AudioData, parameterName=Mid, sortPriority=2}
  ) {group=AudioData, parameterName=MidHigh, sortPriority=3}
  FunctionOutput(None=StaticSwitchParameter(
    True=MaterialFunctionCall!(DebugNoisePanSpeed) {materialFunction=&/MF_MusicDebugNoise}.Noise
    False=StaticSwitchParameter(True=@n1, False=@n4) {group=AudioData, parameterName=High, sortPriority=4}
   ) {group=AudioData, parameterName=UseDebugNoise, sortPriority=5})
 ]}
