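# Masked material for a Creative (Playgrounds) equalizer prop that lights up and bounces with the level from MF_EqSwitches.
# That level drives stepped UV bar fill, emissive color and world position offset on vertices selected by UV1.
# A distance-faded dot grid modulates base color and emissive, while local normal Z separates face looks.
# Static switches add a debug value readout and dithered near-camera fade, before output passes through MF_CreativeMoveTool.
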Material /Game/Playgrounds/Items/EQ/Materials/M_EQ_02
graph {
 properties {
  bUseMaterialAttributes=true
  blendMode=BLEND_Masked
  ditherOpacityMask=true
  opacityMaskClipValue=0.5
  pixelDepthOffsetMode=PDOM_Legacy
  usedShadingModels=Default Lit
 }
 nodes [
  n1 = ScalarParameter() {defaultValue=0.8, parameterName=DotsBaseColorLow}
  n2 = Multiply(A=TextureCoordinate(), B=ScalarParameter() {defaultValue=100, parameterName=DotsRepeat})
  n3 = LinearInterpolate(
   A=SphereMask(
    A=Frac(None=@n2)
    B=Constant() {r=0.5}
    Radius=ScalarParameter() {defaultValue=0.8, parameterName=DotsRadiusLarge}
    Hardness=ScalarParameter() {parameterName=DotsHardness}
   )
   Alpha=MaterialFunctionCall(
    Fade_Start_Distance=ScalarParameter() {defaultValue=800, parameterName="Fade_Start_Distance (S)"}
    Fade_End_Distance=ScalarParameter() {defaultValue=600, parameterName="Fade_End_Distance (S)"}
   ) {materialFunction=&MF_Distance_SyncedFade_FX}.Result
  )
  n4 = ScalarParameter() {defaultValue=0.75, parameterName=RemapMin}
  n5 = ScalarParameter() {defaultValue=0.9, parameterName=RemapMax}
  n6 = MaterialFunctionCall() {materialFunction=&/MF_EqSwitches}
  n7 = LinearInterpolate(
   A=VectorParameter() {parameterName=EmissiveInnerLow, defaultValue=<rgba 0.0081 0.0026 0.0156 0>}.RGB
   B=VectorParameter() {parameterName=EmissiveInnerHigh, defaultValue=<rgba 2 0.3652 0.0382 0>}.RGB
   Alpha=Saturate(None=MaterialFunctionCall!(min=@n4, max=@n5, value=@n6.Result) {materialFunction=/Game/Effects/Fort_Effects/Materials/_Functions/InverseLerp_MF}.Clamped)
  )
  n8 = VectorParameter() {parameterName=UVBarsTiling, defaultValue=<rgba 1 8 0 0>}
  n9 = Multiply(A=@n8.G, B=@n6.Result)
  n10 = SphereMask(
   A=MaterialFunctionCall(Float2=TextureCoordinate()) {materialFunction=&BreakOutFloat2Components}.G
   B=Constant() {r=1}
   Radius=Divide(A=Ceil(None=@n9), B=@n8.G)
   Hardness=Constant()
  )
  n11 = OneMinus(None=@n10)
  n12 = ScalarParameter() {defaultValue=1.5, parameterName=InnerColorLerpExponent}
  n13 = ScalarParameter() {defaultValue=0.5, parameterName=Fresnel Exponent}
  n14 = LinearInterpolate(
   A=VectorParameter() {parameterName=EQLow, defaultValue=<rgba 0.1233 0 0.1233 0>}.RGB
   B=VectorParameter() {parameterName=EQHigh, defaultValue=<rgba 0.4427 1.6755 3 0>}.RGB
   Alpha=Multiply(
    A=Power(Base=@n11, Exp=@n12)
    B=MaterialFunctionCall(Exponent=@n13) {materialFunction=&MF_Fresnel_FX}.Fresnel
   )
  )
  n15 = TextureCoordinate()
  n16 = ComponentMask(None=@n8.RGB) {g=true, r=true}
  n17 = ConstantBiasScale(None=Frac(None=Multiply(A=@n15, B=@n16))) {bias=-0.5, scale=2}
  n18 = ComponentMask(None=Add(A=Abs(None=@n17)) {constB=-0.4}) {g=true}
  n19 = Round(None=@n18)
  n20 = Saturate(None=@n18)
  n21 = Multiply(A=Multiply(A=OneMinus(None=@n19), B=OneMinus(None=@n20)), B=Ceil(None=@n10))
  n22 = Constant() {r=0.9}
  n23 = Constant()
  n24 = Normalize(VectorInput=Transform(None=VertexNormalWS()) {transformSourceType=TRANSFORMSOURCE_World, transformType=TRANSFORM_Local})
  n25 = Abs(None=ConstantBiasScale(None=TextureCoordinate()) {bias=-0.5, scale=2})
  n26 = ScalarParameter() {defaultValue=10, parameterName=Exp}
  n27 = LinearInterpolate(
   A=VectorParameter() {parameterName=EmissiveOuter, defaultValue=<rgba 0.0043 0.0148 0.1059 0>}.RGB
   B=LinearInterpolate(
    A=LinearInterpolate(A=@n7, B=@n14, Alpha=@n21)
    B=If(A=@n22, B=@n6.Result, "A > B"=@n23, "A == B"=@n7, "A < B"=@n7)
    Alpha=MaterialFunctionCall(Float3=@n24) {materialFunction=&BreakOutFloat3Components}.B
   ) {constB=0}
   Alpha=SphereMask(
    A=Power(Base=@n25, Exp=@n26)
    B=ScalarParameter() {parameterName=Center}
    Radius=ScalarParameter() {defaultValue=0.75, parameterName=Radius}
    Hardness=ScalarParameter() {defaultValue=0.6, parameterName=Hardness}
   )
  )
  n28 = MaterialFunctionCall(Number=@n6.Result) {materialFunction=&DebugScalarValues}
  n29 = Constant()
  n30 = MaterialFunctionCall(Float2=TextureCoordinate() {coordinateIndex=1}) {materialFunction=&BreakOutFloat2Components}
  n31 = Multiply(A=Round(None=OneMinus(None=@n30.G)), B=VectorParameter() {parameterName=BounceUpOnWorldZ, defaultValue=<rgba 0 0 96 0>}.RGB)
  n32 = ScalarParameter() {defaultValue=1, parameterName=Opacity}
  n33 = ScalarParameter() {defaultValue=200, parameterName=NearCameraFadeStart}
  n34 = ScalarParameter() {defaultValue=160, parameterName=NearCameraFadeEnd}
  n35 = SetMaterialAttributes!(
   MaterialAttributes
   Base Color=Multiply(A=VectorParameter() {parameterName=BaseColor, defaultValue=<rgba 0.0883 0.0803 0.1476 0>}.RGB, B=LinearInterpolate(A=@n1, Alpha=@n3))
   Metallic=Multiply(
    A=ScalarParameter() {defaultValue=0.8, parameterName=Metallic}
    B=TextureSample() {samplerType=SAMPLERTYPE_Masks, texture=/Game/Environments/Sets/Arid/Textures/T_desert_sand_base_S}.B
   )
   Specular=ScalarParameter() {defaultValue=0.3, parameterName=Specular}
   Roughness=ScalarParameter() {defaultValue=0.2, parameterName=Roughness}
   Emissive Color=Add(A=Multiply(A=@n3, B=@n27), B=StaticSwitchParameter(True=@n28.Result, False=@n29) {parameterName=DisplayEQValues}) {desc=To Emissive}
   World Position Offset=Transform(None=Multiply(A=@n6.Result, B=@n31)) {transformSourceType=TRANSFORMSOURCE_Local}
   Opacity Mask=StaticSwitchParameter(
    True=MaterialFunctionCall(Opacity In=@n32, Fade_Start_Distance=@n33, Fade_End_Distance=@n34) {materialFunction=&MF_Distance_SyncedFade_FX}.Result
    False=@n32
   ) {parameterName=UseNearCameraFade}
  ) {attributeSetTypes [
    #1
    #2
    #3
    #4
    #5
    #6
    #7
   ]}
  MaterialFunctionCall(MA In=@n35) {materialFunction=&MF_CreativeMoveTool}
 ]
}
