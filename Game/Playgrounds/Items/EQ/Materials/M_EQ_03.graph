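# Masked, lit material for a Playgrounds EQ item, with a gradient base color and pulsing blocky glow.
# A UV-derived face mask and sphere mask confine emissive blocks from a world-aligned transition texture, animated by MF_EqSwitches.
# The same switch value scales glow and lifts vertices selected by UV1 along a local-space offset.
# Opacity supports dithered near-camera fade, and MF_CreativeMoveTool wraps the final attributes.
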Material /Game/Playgrounds/Items/EQ/Materials/M_EQ_03
graph {
 properties {bUseMaterialAttributes=true, blendMode=BLEND_Masked, ditherOpacityMask=true, pixelDepthOffsetMode=PDOM_Legacy, usedShadingModels=Default Lit}
 nodes [
  n1 = OneMinus(None=MaterialFunctionCall(Float2=TextureCoordinate()) {materialFunction=&BreakOutFloat2Components}.G)
  n2 = Abs(None=ConstantBiasScale(None=TextureCoordinate()) {bias=-0.5, scale=2})
  n3 = Round(None=Subtract(A=@n2, B=Constant2Vector() {g=0.49, r=0.45}) {constB=0.45})
  n4 = Constant() {r=1}
  n5 = OneMinus(None=Saturate(None=DotProduct(A=@n3, B=@n4)))
  n6 = SphereMask(
   A=Power(Base=@n2, Exp=ScalarParameter() {defaultValue=2, parameterName=Exp})
   B=ScalarParameter() {parameterName=Center}
   Radius=ScalarParameter() {defaultValue=1.2, parameterName=Radius}
   Hardness=ScalarParameter() {defaultValue=0.2, parameterName=Hardness}
  )
  n7 = ScalarParameter() {defaultValue=1, parameterName=BlocksLow}
  n8 = ScalarParameter() {defaultValue=1.2, parameterName=BlocksHigh}
  n9 = TextureSample(UVs=MaterialFunctionCall!(WorldScale=VectorParameter() {parameterName=WorldScale, defaultValue=<rgba 128 128 96 0>}.RGB, TransformToLocalSpace) {materialFunction=/Game/Items/Traps/Materials/MF_Creative_WorldAlignedUVs}.Result) {texture=/Game/Weapons/Prototype/ObjectMover/Textures/T_Movetool_Transition}
  n10 = MaterialFunctionCall() {materialFunction=&/MF_EqSwitches}
  n11 = Multiply(A=OneMinus(None=@n9.R), B=Sine(None=Add(A=@n9.G, B=@n10.Result)) {period=0.45})
  n12 = Multiply(
   A=Multiply(A=VectorParameter() {parameterName=GlowColor, defaultValue=<rgba 0.55 0.0019 0.3226 0>}.RGB, B=LinearInterpolate(A=@n7, B=@n8, Alpha=@n11))
   B=MaterialFunctionCall!(
    min=ScalarParameter() {defaultValue=0.6, parameterName=RemapMin}
    max=ScalarParameter() {defaultValue=0.9, parameterName=RemapMax}
    value=@n10.Result
   ) {materialFunction=/Game/Effects/Fort_Effects/Materials/_Functions/InverseLerp_MF}.Clamped
  )
  n13 = TextureCoordinate() {coordinateIndex=1}
  n14 = Round(None=OneMinus(None=MaterialFunctionCall(Float2=@n13) {materialFunction=&BreakOutFloat2Components}.G))
  n15 = Multiply(A=@n10.Result, B=VectorParameter() {parameterName=BounceUpOnWorldZ, defaultValue=<rgba 0 0 230 0>}.RGB)
  n16 = ScalarParameter() {defaultValue=1, parameterName=Opacity}
  n17 = ScalarParameter() {defaultValue=200, parameterName=NearCameraFadeStart}
  n18 = ScalarParameter() {defaultValue=160, parameterName=NearCameraFadeEnd}
  n19 = SetMaterialAttributes!(
   MaterialAttributes
   Base Color=LinearInterpolate(
    A=VectorParameter() {parameterName=BaseColorLow, defaultValue=<rgba 0.3906 0.0241 0.1498 0>}.RGB
    B=VectorParameter() {parameterName=BaseColorHigh, defaultValue=<rgba 0.7517 0.0431 0.1276 0>}.RGB
    Alpha=Multiply(A=@n1, B=@n5)
   )
   Metallic=ScalarParameter() {defaultValue=0.4, parameterName=Metallic}
   Specular=ScalarParameter() {defaultValue=0.3, parameterName=Specular}
   Roughness=ScalarParameter() {defaultValue=0.2, parameterName=Roughness}
   Emissive Color=Multiply(A=@n5, B=Multiply(A=@n6, B=@n12))
   World Position Offset=Transform(None=Multiply(A=@n14, B=@n15)) {transformSourceType=TRANSFORMSOURCE_Local}
   Opacity Mask=StaticSwitchParameter(
    True=MaterialFunctionCall(Opacity In=@n16, Fade_Start_Distance=@n17, Fade_End_Distance=@n18) {materialFunction=&MF_Distance_SyncedFade_FX}.Result
    False=@n16
   ) {defaultValue=true, parameterName=UseNearCameraFade}
  ) {attributeSetTypes [
    #1
    #2
    #3
    #4
    #5
    #6
    #7
   ]}
  MaterialFunctionCall(MA In=@n19) {materialFunction=&MF_CreativeMoveTool}
 ]
}
