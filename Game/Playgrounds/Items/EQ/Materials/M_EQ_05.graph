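# Masked, lit metallic material for an equalizer (EQ) item, with a glowing dot that responds to Creative Jukebox audio.
# AverageAmplitude from the Jukebox parameter collection is inverse-lerped between Min and Max range, blending a UV sphere-mask dot from dim EmissiveOuter to bright Glow_Color.
# A powered, centered-UV mask adds a constant rounded-square edge glow.
# Attributes pass through MF_CreativeMoveTool, with opacity set by a dithered scalar.
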Material /Game/Playgrounds/Items/EQ/Materials/M_EQ_05
graph {
 properties {
  bUseMaterialAttributes=true
  blendMode=BLEND_Masked
  ditherOpacityMask=true
  opacityMaskClipValue=0.5
  pixelDepthOffsetMode=PDOM_Legacy
  usedShadingModels=Default Lit
 }
 nodes [
  n1 = SphereMask(
   A=TextureCoordinate()
   B=Constant() {r=0.5}
   Radius=ScalarParameter() {defaultValue=0.6544, parameterName=DotsRadiusLarge}
   Hardness=ScalarParameter() {parameterName=DotsHardness}
  )
  n2 = Multiply(
   A=@n1
   B=Multiply(
    A=VectorParameter() {parameterName=Glow_Color, defaultValue=<rgba 150 150 150 1>}.RGB
    B=ScalarParameter() {defaultValue=800, parameterName=Glow_Mult, sortPriority=1}
   )
  )
  n3 = MaterialFunctionCall!(
   min=ScalarParameter() {parameterName=Min_Range}
   max=ScalarParameter() {defaultValue=0.2, parameterName=Max_Range}
   value=CollectionParameter() {collection=/Game/Creative/Devices/Jukebox/Creative_Jukebox_MPC, parameterName=AverageAmplitude}
  ) {materialFunction=/Game/Effects/Fort_Effects/Materials/_Functions/InverseLerp_MF}
  n4 = VectorParameter() {parameterName=EmissiveOuter, defaultValue=<rgba 0.0043 0.0148 0.1059 0>}
  n5 = Multiply(A=@n1, B=@n4.RGB)
  n6 = Abs(None=ConstantBiasScale(None=TextureCoordinate()) {bias=-0.5, scale=2})
  n7 = ScalarParameter() {defaultValue=20, parameterName=Exp}
  n8 = OneMinus(None=SphereMask(
    A=Power(Base=@n6, Exp=@n7)
    B=ScalarParameter() {parameterName=Center}
    Radius=ScalarParameter() {defaultValue=0.75, parameterName=Radius}
    Hardness=ScalarParameter() {defaultValue=0.6, parameterName=Hardness}
   ))
  n9 = SetMaterialAttributes!(
   MaterialAttributes
   Base Color=VectorParameter() {parameterName=BaseColor, defaultValue=<rgba 0.0883 0.0803 0.1476 0>}.RGB
   Metallic=Multiply(
    A=ScalarParameter() {defaultValue=0.8, parameterName=Metallic}
    B=TextureSample() {samplerType=SAMPLERTYPE_Masks, texture=/Game/Environments/Sets/Arid/Textures/T_desert_sand_base_S}.B
   )
   Specular=ScalarParameter() {defaultValue=0.3, parameterName=Specular}
   Roughness=ScalarParameter() {defaultValue=0.2, parameterName=Roughness}
   Emissive Color=Add(A=LinearInterpolate(A=@n5, B=@n2, Alpha=@n3.Clamped), B=Multiply(A=@n4.RGB, B=@n8))
   World Position Offset
   Opacity Mask=ScalarParameter() {defaultValue=1, parameterName=Opacity}
  ) {attributeSetTypes [
    #1
    #2
    #3
    #4
    #5
    #6
    #7
   ]}
  MaterialFunctionCall(MA In=@n9) {materialFunction=&MF_CreativeMoveTool}
 ]
}
