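# Material function for EQ visuals that turns the five Creative Jukebox frequency band levels into one value varying along the V texture coordinate.
# It reads Low, LowMid, Mid, MidHigh and High from the Jukebox parameter collection and chains lerps between them at unit steps.
# The default coordinate is UV1 with V tiled to span the bands.
# The first lerp's alpha is unclamped, unlike the others, so it extrapolates outside its segment.
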MaterialFunction /Game/Playgrounds/Items/EQ/Materials/MF_EQ_5Bands
graph {nodes [
  n1 = MaterialFunctionCall(Float2=FunctionInput(Preview=TextureCoordinate() {coordinateIndex=1, vTiling=4}) {bUsePreviewValueAsDefault=true, inputName=TexCoord, inputType=FunctionInput_Vector2}) {materialFunction=&BreakOutFloat2Components}
  n2 = LinearInterpolate(
   A=LinearInterpolate(
    A=CollectionParameter() {collection=/Game/Creative/Devices/Jukebox/Creative_Jukebox_MPC, parameterName=Low}
    B=CollectionParameter() {collection=&Creative_Jukebox_MPC, parameterName=LowMid}
    Alpha=@n1.G
   )
   B=CollectionParameter() {collection=&Creative_Jukebox_MPC, parameterName=Mid}
   Alpha=Saturate(None=Add(A=@n1.G) {constB=-1})
  )
  n3 = CollectionParameter() {collection=&Creative_Jukebox_MPC, parameterName=MidHigh}
  n4 = Saturate(None=Add(A=@n1.G) {constB=-2})
  n5 = Add(A=@n1.G) {constB=-3}
  FunctionOutput(None=LinearInterpolate(
    A=LinearInterpolate(A=@n2, B=@n3, Alpha=@n4)
    B=CollectionParameter() {collection=&Creative_Jukebox_MPC, parameterName=High}
    Alpha=Saturate(None=@n5)
   ))
 ]}
