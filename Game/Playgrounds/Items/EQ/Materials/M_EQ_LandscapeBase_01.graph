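# Additive, unlit, two-sided material for an EQ landscape effect: a glowing dotted surface whose height and brightness follow MF_EQ_5Bands.
# Band values, perturbed by panning noise offset by actor position and weighted along U, drive dot size, peak emissive boost and scale-aware World Position Offset.
# Noise and caustic plasma layers form the colour, faded by edge mask, depth and camera proximity.
# The result passes through MF_CreativeMoveTool.
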Material /Game/Playgrounds/Items/EQ/Materials/M_EQ_LandscapeBase_01
graph {
 properties {
  allowTranslucentCustomDepthWrites=true
  bEnableResponsiveAA=true
  bUseMaterialAttributes=true
  blendMode=BLEND_Additive
  pixelDepthOffsetMode=PDOM_Legacy
  shadingModel=MSM_Unlit
  twoSided=true
  usedShadingModels=Unlit
 }
 nodes [
  n1 = TextureCoordinate()
  n2 = ScalarParameter() {defaultValue=100, parameterName=DotsRepeat}
  n3 = DotProduct(A=Truncate!(None=ActorPositionWS()) {is=class MaterialExpressionTruncate}, B=Constant() {r=0.2})
  n4 = Add(A=@n3, B=TextureCoordinate())
  n5 = VectorParameter() {parameterName=NoiseTilePanSpeed, defaultValue=<rgba 0.122 0.3 0.3 0.03>}
  n6 = AppendVector(A=@n5.R, B=@n5.R)
  n7 = TextureSample(UVs=Panner(Coordinate=Multiply(A=@n4, B=@n6), Speed=AppendVector(A=@n5.B, B=@n5.B)) {bFractionalPart=true, speedX=0.2, speedY=0.05}) {samplerType=SAMPLERTYPE_LinearColor, texture=/Game/Global/Textures/Default/Noise/T_Noise01_Inverse}
  n8 = Subtract(A=@n7.R) {constB=0.5}
  n9 = MaterialFunctionCall!(TexCoord=TextureCoordinate()) {materialFunction=&/MF_EQ_5Bands}
  n10 = LinearInterpolate(
   A=ScalarParameter() {defaultValue=0.3, parameterName=NoiseScaleLow}
   B=ScalarParameter() {defaultValue=3, parameterName=NoiseScaleHigh}
   Alpha=@n9.Result
  )
  n11 = ComponentMask(None=TextureCoordinate()) {r=true}
  n12 = Multiply(A=Add(A=Multiply(A=@n8, B=@n10), B=@n9.Result), B=@n11)
  n13 = SphereMask(
   A=Frac(None=Multiply(A=@n1, B=@n2))
   B=Constant() {r=0.5}
   Radius=LinearInterpolate(
    A=ScalarParameter() {defaultValue=0.1, parameterName=RadiusSmall}
    B=ScalarParameter() {defaultValue=0.2, parameterName=RadiusLarge}
    Alpha=@n12
   )
   Hardness=ScalarParameter() {defaultValue=0.5, parameterName=DotsHardness}
  )
  n14 = VectorParameter() {parameterName=DotsColor, defaultValue=<rgba 1.1641 0.2354 2 0>}
  n15 = VectorParameter() {parameterName=EmissiveA, defaultValue=<rgba 0.1551 0.1494 1 0>}
  n16 = Add(A=TextureCoordinate(), B=@n3)
  n17 = VectorParameter() {parameterName=PlasmaTilePanSpeed, defaultValue=<rgba 0.125 0.5 0.2667 0.2>}
  n18 = AppendVector(A=@n17.R, B=@n17.R)
  n19 = TextureSample(UVs=Panner(Coordinate=Multiply(A=@n16, B=@n18), Speed=AppendVector(A=@n17.B, B=@n17.B)) {bFractionalPart=true, speedX=0.45}) {texture=&T_ShieldCaustics}
  n20 = LinearInterpolate(
   A=Multiply(A=VectorParameter() {parameterName=EmissiveB, defaultValue=<rgba 0.5 0.0402 0.4085 0>}.RGB, B=@n7.R)
   B=LinearInterpolate(A=Multiply(A=@n15.RGB, B=@n19.R), B=@n15.RGB, Alpha=@n11)
   Alpha=@n11
  )
  n21 = Multiply(A=@n20, B=ScalarParameter() {defaultValue=2, parameterName=EmissiveScalePeaks})
  n22 = Power(Base=@n12, Exp=ScalarParameter() {defaultValue=3, parameterName=EmissivePeaksLerpExponent})
  n23 = ConstantBiasScale(None=TextureCoordinate()) {bias=-0.5, scale=2}
  n24 = SphereMask(
   A=Power(Base=Abs(None=@n23), Exp=ScalarParameter() {defaultValue=3, parameterName=Exp})
   B=Constant()
   Radius=ScalarParameter() {defaultValue=1, parameterName=Radius}
   Hardness=ScalarParameter() {parameterName=Hardness}
  )
  n25 = Multiply(
   A=Add(A=Multiply(A=@n13, B=@n14.RGB), B=LinearInterpolate(A=@n20, B=@n21, Alpha=@n22))
   B=DepthFade(Opacity=Multiply(A=@n24, B=@n11), FadeDistance=ScalarParameter() {defaultValue=30, parameterName=FadeDistance})
  )
  n26 = MaterialFunctionCall(Distance=ScalarParameter() {defaultValue=70, parameterName=NearFadeDistance}) {materialFunction=&Near_Camera_Fade}
  n27 = MaterialFunctionCall() {materialFunction=/Engine/Functions/Engine_MaterialFunctions02/WorldPositionOffset/ObjectScale}
  n28 = Multiply(A=@n12, B=Multiply(A=VectorParameter() {parameterName=WPOHeight, defaultValue=<rgba 0 0 70 0>}.RGB, B=Power(Base=@n27.Scale XYZ)))
  MaterialFunctionCall(MA In=SetMaterialAttributes!(
    MaterialAttributes
    Emissive Color=Multiply(A=@n25, B=@n26.Result)
    World Position Offset=Transform(None=@n28) {transformSourceType=TRANSFORMSOURCE_World, transformType=TRANSFORM_Local}
   ) {attributeSetTypes [#1, #2]}) {materialFunction=&MF_CreativeMoveTool}
 ]
}
