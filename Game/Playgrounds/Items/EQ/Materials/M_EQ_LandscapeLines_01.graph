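# Unlit additive material for EQ landscape lines, drawing glowing bars that rise and thicken with a five-band EQ function.
# MF_EQ_5Bands output plus panning noise forms a level scaled along UV1, which drives object-scaled height offset and SplineThicken width.
# The remapped level also blends emissive colors and sharpness of the sine-wobbled bar mask.
# Sphere masks shape opacity, softened by depth fade and near-camera fade.
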Material /Game/Playgrounds/Items/EQ/Materials/M_EQ_LandscapeLines_01
graph {
 properties {
  allowTranslucentCustomDepthWrites=true
  bEnableResponsiveAA=true
  blendMode=BLEND_Additive
  pixelDepthOffsetMode=PDOM_Legacy
  shadingModel=MSM_Unlit
  usedShadingModels=Unlit
 }
 nodes [
  n1 = TextureCoordinate() {coordinateIndex=1}
  n2 = VectorParameter() {parameterName=NoiseTilePanSpeed, defaultValue=<rgba 0.122 0.3 0.3 0.03>}
  n3 = Multiply(A=@n1, B=AppendVector(A=@n2.R, B=@n2.R))
  n4 = AppendVector(A=@n2.B, B=@n2.B)
  n5 = Subtract(A=TextureSample(UVs=Panner(Coordinate=@n3, Speed=@n4) {bFractionalPart=true, speedX=0.2, speedY=0.05}) {samplerType=SAMPLERTYPE_LinearColor, texture=/Game/Global/Textures/Default/Noise/T_Noise01_Inverse}.R) {constB=0.5}
  n6 = MaterialFunctionCall!(TexCoord) {materialFunction=&/MF_EQ_5Bands}
  n7 = LinearInterpolate(
   A=ScalarParameter() {defaultValue=0.3, parameterName=NoiseScaleLow}
   B=ScalarParameter() {defaultValue=3, parameterName=NoiseScaleHigh}
   Alpha=@n6.Result
  )
  n8 = Multiply(A=Add(A=Multiply(A=@n5, B=@n7), B=@n6.Result), B=ComponentMask(None=TextureCoordinate() {coordinateIndex=1}) {r=true})
  n9 = MaterialFunctionCall() {materialFunction=/Engine/Functions/Engine_MaterialFunctions02/WorldPositionOffset/ObjectScale}
  n10 = Multiply(A=VectorParameter() {parameterName=WPOHeight, defaultValue=<rgba 0 0 80 0>}.RGB, B=Power(Base=@n9.Scale XYZ))
  n11 = MaterialFunctionCall!(
   min=ScalarParameter() {defaultValue=0.2, parameterName=RemapLow}
   max=ScalarParameter() {defaultValue=0.85, parameterName=remapHigh}
   value=@n8
  ) {materialFunction=/Game/Effects/Fort_Effects/Materials/_Functions/InverseLerp_MF}
  n12 = LinearInterpolate(
   A=ScalarParameter() {defaultValue=2, parameterName=WidthBaseLow}
   B=ScalarParameter() {defaultValue=2.5, parameterName=WidthBaseHigh}
   Alpha=@n11.Clamped
  )
  Add(
   A=Transform(None=Multiply(A=@n8, B=@n10)) {transformSourceType=TRANSFORMSOURCE_World, transformType=TRANSFORM_Local}
   B=MaterialFunctionCall(WidthBase=ScalarParameter() {defaultValue=1, parameterName=WidthTip}, WidthTip=Multiply(A=@n9.Scale Z, B=@n12)) {materialFunction=&SplineThicken}.WorldPositionOffset
  )
  n13 = ActorPositionWS()
  Add(A=DotProduct(A=Truncate!(None=@n13) {is=class MaterialExpressionTruncate}, B=Constant() {r=0.2}), B=@n1)
  n14 = ComponentMask(None=TextureCoordinate() {coordinateIndex=1}) {r=true}
  LinearInterpolate(
   A=LinearInterpolate(
    A=VectorParameter() {parameterName=EmissiveLowA, defaultValue=<rgba 0.3141 2.5 1.1783 0>}.RGB
    B=VectorParameter() {parameterName=EmissiveLowB, defaultValue=<rgba 0.2902 0.0304 2.5 0>}.RGB
    Alpha=@n14
   )
   B=VectorParameter() {parameterName=EmissiveHigh, defaultValue=<rgba 0.5909 0.3814 10 0>}.RGB
   Alpha=@n11.Clamped
  )
  n15 = ComponentMask(None=TextureCoordinate()) {r=true}
  n16 = ComponentMask(None=Panner(Coordinate=TextureCoordinate()) {bFractionalPart=true, speedY=0.42}) {g=true}
  n17 = Multiply(A=Multiply(A=Sine(None=@n16) {period=0.025}, B=ScalarParameter() {defaultValue=0.15, parameterName=BarNoiseScale}), B=@n6.Result)
  n18 = Power(
   Base=SphereMask(
    A=Add(A=@n15, B=@n17) {constB=0}
    B=Constant() {r=0.5}
    Radius=ScalarParameter() {defaultValue=0.6, parameterName=BarRadius}
    Hardness=ScalarParameter() {defaultValue=0.1, parameterName=BarHardness}
   )
   Exp=LinearInterpolate(
    A=ScalarParameter() {defaultValue=3, parameterName=MaskExponentLow}
    B=ScalarParameter() {defaultValue=2, parameterName=MaskExponentHigh}
    Alpha=@n11.Clamped
   )
  )
  n19 = SphereMask(A=Power(Base=@n14) {constExponent=4}, B=Constant() {r=0.5}, Radius=Constant() {r=0.5}, Hardness=Constant() {r=0.5})
  Multiply(
   A=DepthFade(Opacity=Multiply(A=@n18, B=@n19), FadeDistance=ScalarParameter() {defaultValue=15, parameterName=DepthFadeDistance})
   B=MaterialFunctionCall(Distance=ScalarParameter() {defaultValue=70, parameterName=NearFadeDistance}) {materialFunction=&Near_Camera_Fade}.Result
  )
 ]
}
